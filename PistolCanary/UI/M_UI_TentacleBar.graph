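# Translucent UI material for the "TentacleBar", a segmented progress bar with a two-color gradient fill.
# Custom polygon SDF nodes, antialiased with screen-space derivatives, build tapered base, bottom band and upper triangle masks.
# MF_UI_Sectors creates a segmentation mask from a segment-count parameter, and a Progress step cuts a progress mask along horizontal UV.
# Named reroutes carry these masks into multiply and subtract combinations, including a Fillbar mask.
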Material /PistolCanary/UI/M_UI_TentacleBar
graph {
 properties {bAutomaticallySetUsageInEditor=false, blendMode=BLEND_Translucent, materialDomain=MD_UI, usedShadingModels=Default Lit}
 nodes [
  LinearInterpolate(
   A=VectorParameter() {parameterName=Fill Color 1, defaultValue=<rgba 1 0.2755 0 1>}.RGB
   B=VectorParameter() {parameterName=Fill Color 2, defaultValue=<rgba 0 0.8768 1 1>}.RGB
   Alpha=MaterialFunctionCall(UVs=TextureCoordinate(), GradientStart=Constant() {r=0.1}, GradientEnd=Constant() {r=0.9}) {materialFunction=&MF_UI_LinearGradient}.Result
  )
  Multiply(A=NamedRerouteUsage(), B=NamedRerouteUsage())
  n1 = MaterialFunctionCall!(
   UVs
   Gradient Map=ComponentMask(None=TextureCoordinate()) {r=true}
   Sector Count=ScalarParameter() {defaultValue=3, parameterName=Number of Segments}
   Sector Spacing=Constant() {r=0.97}
   Fixed Spacing
  ) {materialFunction=/FortUILibrary/MaterialFunctions/Basics/MF_UI_Sectors}
  NamedRerouteDeclaration(None=OneMinus(None=Step(X=@n1.Convergent Sectors SDF) {constY=0.1})) {name=Segmentation, nodeColor=<rgba 0 0 0 1>}
  n2 = MaterialFunctionCall(Float2=MaterialFunctionCall() {materialFunction=&GetUserInterfaceUV}.Pixel Size) {materialFunction=&BreakOutFloat2Components}
  n3 = Constant() {r=1}
  n4 = Max(A=AppendVector(A=Divide(A=@n2.R, B=@n2.R), B=@n3) {desc=Horizontal}, B=AppendVector(A=@n3, B=Divide(A=@n2.G, B=@n2.G)) {desc=Vertical})
  MaterialExpressionMultiply_0 = Multiply(A=@n4, B=TextureCoordinate())
  MaterialExpressionMultiply_1 = Multiply(A=@n4, B=Constant2Vector())
  MaterialExpressionMultiply_2 = Multiply(A=@n4, B=Constant2Vector() {r=1})
  MaterialExpressionMultiply_3 = Multiply(A=@n4, B=Constant2Vector() {g=1, r=0.95})
  MaterialExpressionMultiply_4 = Multiply(A=@n4, B=Constant2Vector() {g=1, r=0.05})
  n5 = Custom!(
   p=@MaterialExpressionMultiply_0
   v1=@MaterialExpressionMultiply_1
   v2=@MaterialExpressionMultiply_2
   v3=@MaterialExpressionMultiply_3
   v4=@MaterialExpressionMultiply_4
  ) {
   code="float2 v[] = {v1,v2,v3,v4};\r\nconst int N = 4;\r\n\r\nfloat d = dot(p-v[0],p-v[0]);\r\nfloat s = 2.0;\r\nfor( int i=0, j=N-1; i<N; j=i, i++ )\r\n{\r\n    float2 e = v[j] - v[i];\r\n    float2 w = p - v[i];\r\n    float2 b = w - e*clamp( dot(w,e)/dot(e,e), 0.0, 1.0 );\r\n    d = min( d, dot(b,b) );\r\n    float3 c = float3(p.y>=v[i].y,p.y<v[j].y,e.x*w.y>e.y*w.x);\r\n    if( all(c) || all(!c) ) s = -s;\r\n}\r\nreturn s*sqrt(d);"
   description=PolySDF
   outputType=CMOT_Float1
   inputs [
    {inputName=p, input {expression=&self:MaterialExpressionMultiply_0}}
    {inputName=v1, input {expression=&self:MaterialExpressionMultiply_1}}
    {inputName=v2, input {expression=&self:MaterialExpressionMultiply_2}}
    {inputName=v3, input {expression=&self:MaterialExpressionMultiply_3}}
    {inputName=v4, input {expression=&self:MaterialExpressionMultiply_4}}
   ]
  }
  n6 = AppendVector(A=DDX(Value=@n5), B=DDY(Value=@n5))
  n7 = Divide(A=@n5, B=Add(A=Length(None=@n6) {is=class MaterialExpressionLength}) {constB=0})
  NamedRerouteDeclaration(None=OneMinus(None=Saturate(None=@n7))) {name=BaseShape, nodeColor=<rgba 0 0 0 1>}
  n8 = MaterialFunctionCall(Float2=MaterialFunctionCall() {materialFunction=&GetUserInterfaceUV}.Pixel Size) {materialFunction=&BreakOutFloat2Components}
  n9 = Constant() {r=1}
  n10 = Max(A=AppendVector(A=Divide(A=@n8.R, B=@n8.R), B=@n9) {desc=Horizontal}, B=AppendVector(A=@n9, B=Divide(A=@n8.G, B=@n8.G)) {desc=Vertical})
  MaterialExpressionMultiply_5 = Multiply(A=@n10, B=TextureCoordinate())
  MaterialExpressionMultiply_6 = Multiply(A=@n10, B=Constant2Vector() {g=0.55, r=0.05})
  MaterialExpressionMultiply_7 = Multiply(A=@n10, B=Constant2Vector() {g=0.55, r=0.95})
  MaterialExpressionMultiply_8 = Multiply(A=@n10, B=Constant2Vector() {g=1, r=0.95})
  MaterialExpressionMultiply_9 = Multiply(A=@n10, B=Constant2Vector() {g=1, r=0.05})
  n11 = Custom(
   p=@MaterialExpressionMultiply_5
   v1=@MaterialExpressionMultiply_6
   v2=@MaterialExpressionMultiply_7
   v3=@MaterialExpressionMultiply_8
   v4=@MaterialExpressionMultiply_9
  ) {
   code="float2 v[] = {v1,v2,v3,v4};\r\nconst int N = 4;\r\n\r\nfloat d = dot(p-v[0],p-v[0]);\r\nfloat s = 2.0;\r\nfor( int i=0, j=N-1; i<N; j=i, i++ )\r\n{\r\n    float2 e = v[j] - v[i];\r\n    float2 w = p - v[i];\r\n    float2 b = w - e*clamp( dot(w,e)/dot(e,e), 0.0, 1.0 );\r\n    d = min( d, dot(b,b) );\r\n    float3 c = float3(p.y>=v[i].y,p.y<v[j].y,e.x*w.y>e.y*w.x);\r\n    if( all(c) || all(!c) ) s = -s;\r\n}\r\nreturn s*sqrt(d);"
   description=PolySDF
   outputType=CMOT_Float1
   inputs [
    {inputName=p, input {expression=&self:MaterialExpressionMultiply_5}}
    {inputName=v1, input {expression=&self:MaterialExpressionMultiply_6}}
    {inputName=v2, input {expression=&self:MaterialExpressionMultiply_7}}
    {inputName=v3, input {expression=&self:MaterialExpressionMultiply_8}}
    {inputName=v4, input {expression=&self:MaterialExpressionMultiply_9}}
   ]
  }
  n12 = AppendVector(A=DDX(Value=@n11), B=DDY(Value=@n11))
  n13 = Divide(A=@n11, B=Add(A=Length(None=@n12)) {constB=0})
  NamedRerouteDeclaration(None=OneMinus(None=Saturate(None=@n13))) {name=BottomShape, nodeColor=<rgba 0 0 0 1>}
  n14 = MaterialFunctionCall(Float2=MaterialFunctionCall() {materialFunction=&GetUserInterfaceUV}.Pixel Size) {materialFunction=&BreakOutFloat2Components}
  n15 = Constant() {r=1}
  n16 = Max(A=AppendVector(A=Divide(A=@n14.R, B=@n14.R), B=@n15) {desc=Horizontal}, B=AppendVector(A=@n15, B=Divide(A=@n14.G, B=@n14.G)) {desc=Vertical})
  MaterialExpressionMultiply_10 = Multiply(A=@n16, B=TextureCoordinate())
  MaterialExpressionMultiply_11 = Multiply(A=@n16, B=Constant2Vector())
  MaterialExpressionMultiply_12 = Multiply(A=@n16, B=Constant2Vector() {r=1})
  MaterialExpressionMultiply_13 = Multiply(A=@n16, B=Constant2Vector() {g=0.2, r=0.5})
  MaterialExpressionMultiply_14 = Multiply(A=@n16, B=Constant2Vector() {g=0.2, r=0.5})
  n17 = Custom(
   p=@MaterialExpressionMultiply_10
   v1=@MaterialExpressionMultiply_11
   v2=@MaterialExpressionMultiply_12
   v3=@MaterialExpressionMultiply_13
   v4=@MaterialExpressionMultiply_14
  ) {
   code="float2 v[] = {v1,v2,v3,v4};\r\nconst int N = 4;\r\n\r\nfloat d = dot(p-v[0],p-v[0]);\r\nfloat s = 2.0;\r\nfor( int i=0, j=N-1; i<N; j=i, i++ )\r\n{\r\n    float2 e = v[j] - v[i];\r\n    float2 w = p - v[i];\r\n    float2 b = w - e*clamp( dot(w,e)/dot(e,e), 0.0, 1.0 );\r\n    d = min( d, dot(b,b) );\r\n    float3 c = float3(p.y>=v[i].y,p.y<v[j].y,e.x*w.y>e.y*w.x);\r\n    if( all(c) || all(!c) ) s = -s;\r\n}\r\nreturn s*sqrt(d);"
   description=PolySDF
   outputType=CMOT_Float1
   inputs [
    {inputName=p, input {expression=&self:MaterialExpressionMultiply_10}}
    {inputName=v1, input {expression=&self:MaterialExpressionMultiply_11}}
    {inputName=v2, input {expression=&self:MaterialExpressionMultiply_12}}
    {inputName=v3, input {expression=&self:MaterialExpressionMultiply_13}}
    {inputName=v4, input {expression=&self:MaterialExpressionMultiply_14}}
   ]
  }
  n18 = AppendVector(A=DDX(Value=@n17), B=DDY(Value=@n17))
  n19 = Divide(A=@n17, B=Add(A=Length(None=@n18)) {constB=0})
  NamedRerouteDeclaration(None=OneMinus(None=Saturate(None=@n19))) {name=UpperShape, nodeColor=<rgba 0 0 0 1>}
  n20 = ComponentMask(None=MaterialFunctionCall() {materialFunction=&GetUserInterfaceUV}.Normalized UV) {r=true}
  n21 = ScalarParameter() {defaultValue=0.5, parameterName=Progress}
  NamedRerouteDeclaration(None=Multiply(A=Step(Y=@n20, X=@n21), B=NamedRerouteUsage())) {name=Progressbar, nodeColor=<rgba 0 0 0 1>}
  NamedRerouteDeclaration(None=Subtract(A=NamedRerouteUsage(), B=NamedRerouteUsage())) {name=Fillbar, nodeColor=<rgba 0 0 0 1>}
 ]
}
